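# Niagara system for the Bass Boost vehicle boost cosmetic, emitting effects from skeletal mesh sockets split by port count.
# Layers three spawn groups of dissolving cone meshes and sprites, tagged for renderer visibility, with ribbons wobbled by sine/cosine offsets and backward-flying trail meshes.
# ColorVariantID arrays and painted-variant overrides tint each layer.
# The frontend-preview flag adjusts offsets, lifetimes and spawn rates, while DTCarOffset shifts spawn positions per car.
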NiagaraSystem /VehicleCosmetics/Boosts/Boost_BassBoost/FX/NS_Boost_BassBoost_OP
dependencies {
 usedDataInterfaces [
  &NiagaraDataInterfaceArrayColor
  &NiagaraDataInterfaceCurve
  &NiagaraDataInterfaceArrayFloat
  &NiagaraDataInterfaceArrayFloat3
  &NiagaraDataInterfaceColorCurve
  &NiagaraDataInterfaceSkeletalMesh
 ]
 usedDynamicInputs []
 usedModules [
  &self:SystemSpawnScript.SystemScriptSource.SystemScriptGraph.NiagaraNodeAssignment_0.SetVariables_#1
  &SystemState
 ]
}
emitters {
 NE_BoostCone_Meshes {
  emitterSpawnScript [/SetVariables(
    Emitter.R1_BaseColorTint_Blue:LinearColor=&SelectLinearColorFromArray(
     Array Sampling Mode=::NewEnumerator1
     Direct Array Index=$System.ColorVariantID
     Color Selection Array=DataInterface() {propertyValues {ColorData [
        <rgba 0 0.5 1.5 1>
        <rgba 1 1 1 1>
        <rgba 0.001 0.001 0.001 1>
        <rgba 0.5 0.1 0 1>
        <rgba 0.1 0.3 1 1>
        <rgba 1 0 0 1>
        <rgba 0 1 0 1>
        <rgba 0.15 0.15 0.15 1>
        <rgba 0.7898 0.85 0.1275 1>
        <rgba 2.5 1.0417 0 1>
        <rgba 2 0.2 0.8 1>
        <rgba 1.1 0.2 2 1>
        <rgba 1.25 1.1458 0 1>
        <rgba 0.06 1.1917 2 1>
        <rgba 1.5 1.5 1.5 1>
       ]}}
     Random Seed=?
    )
    Emitter.R1_CustomFresnelColor_Blue:LinearColor=&SelectLinearColorFromArray(
     Array Sampling Mode=::NewEnumerator1
     Direct Array Index=$System.ColorVariantID
     Color Selection Array=DataInterface() {propertyValues {ColorData [
        <rgba 0.1 0.45 0.8 -0.5>
        <rgba 1 1 1 1>
        <rgba 0.025 0.025 0.025 -0.5>
        <rgba 0.6 0.12 0 -0.5>
        <rgba 0.1 0.3 1 -0.5>
        <rgba 1 0 0 -0.5>
        <rgba 0 1 0 -0.5>
        <rgba 0.15 0.15 0.15 0>
        <rgba 1 1 0.042 0>
        <rgba 0.75 0.3344 0.0375 0>
        <rgba 1 0.1 0.4 0>
        <rgba 0.275 0.05 0.5 0>
        <rgba 0.85 0.7813 0.0255 0>
        <rgba 0.03 0.5958 1 0>
        <rgba 1 1 1 0>
       ]}}
     Random Seed=?
    )
    Emitter.R1_DissolveHighlightColor_Blue:LinearColor=&SelectLinearColorFromArray(
     Array Sampling Mode=::NewEnumerator1
     Direct Array Index=$System.ColorVariantID
     Color Selection Array=DataInterface() {propertyValues {ColorData [
        <rgba 0 2.5 1.5 1>
        <rgba 1 1 1 1>
        <rgba 0.025 0.025 0.025 1>
        <rgba 1 0.4 0.1 1>
        <rgba 0.1 0.3 1 1>
        <rgba 1 0 0 1>
        <rgba 0 1 0 1>
        <rgba 0.25 0.25 0.25 1>
        <rgba 0.1219 0.15 0.0375 1>
        <rgba 0.8 0.2933 0.04 1>
        <rgba 2 0.2 0.8 1>
        <rgba 1.1 0.2 2 1>
        <rgba 1 0.9167 0 1>
        <rgba 0 0.5833 1 1>
        <rgba 0.75 0.75 0.75 1>
       ]}}
     Random Seed=?
    )
    Emitter.R2_BaseColorTint_Green:LinearColor=&SelectLinearColorFromArray(
     Array Sampling Mode=::NewEnumerator1
     Direct Array Index=$System.ColorVariantID
     Color Selection Array=DataInterface() {propertyValues {ColorData [
        <rgba 0 1.5 0.25 1>
        <rgba 1 1 1 1>
        <rgba 0.03 0.03 0.03 1>
        <rgba 1 0.4 0.1 1>
        <rgba 0.1 0.3 1 1>
        <rgba 0.75 0 0 1>
        <rgba 0 0.5 0 1>
        <rgba 0.25 0.25 0.25 1>
        <rgba 0.4167 0.5 0 1>
        <rgba 1.25 0.4583 0.0625 1>
        <rgba 1 0.1 0.4 1>
        <rgba 0.55 0.1 1 1>
        <rgba 0.975 0.8962 0.0292 1>
        <rgba 0 0.5833 1 1>
        <rgba 0.95 0.95 0.95 1>
       ]}}
     Random Seed=?
    )
    Emitter.R2_CustomFesnelColor_Green:LinearColor=&SelectLinearColorFromArray(
     Array Sampling Mode=::NewEnumerator1
     Direct Array Index=$System.ColorVariantID
     Color Selection Array=DataInterface() {propertyValues {ColorData [
        <rgba 0.15 0.85 0.35 -0.5>
        <rgba 1 1 1 1>
        <rgba 0.06 0.06 0.06 -0.5>
        <rgba 1 0.2 0.1 -0.5>
        <rgba 0.1 0.3 1 -0.5>
        <rgba 0.5 0 0 0>
        <rgba 0 0.5 0 0>
        <rgba 0.25 0.25 0.25 -0.5>
        <rgba 0.2669 0.35 0.0175 0>
        <rgba 0.8 0.2933 0.04 0>
        <rgba 1 0.1 0.625 0>
        <rgba 0.165 0.03 0.3 0>
        <rgba 1 0.9208 0.05 0>
        <rgba 0 0.5833 1 0>
        <rgba 1 1 1 0>
       ]}}
     Random Seed=?
    )
    Emitter.R2_DissolveHighlightColor_Green:LinearColor=&SelectLinearColorFromArray(
     Array Sampling Mode=::NewEnumerator1
     Direct Array Index=$System.ColorVariantID
     Color Selection Array=DataInterface() {propertyValues {ColorData [
        <rgba 0 2.5 1.5 1>
        <rgba 1 1 1 1>
        <rgba 0.025 0.025 0.025 1>
        <rgba 1 0.4 0.1 1>
        <rgba 0.1 0.3 1 1>
        <rgba 0.5 0 0 1>
        <rgba 0 0.5 0 1>
        <rgba 0.25 0.25 0.25 1>
        <rgba 0.5906 0.75 0.1125 1>
        <rgba 0.8 0.2933 0.04 1>
        <rgba 1 0.1 0.7 1>
        <rgba 1.4 0.2 2 1>
        <rgba 1 0.9167 0 1>
        <rgba 0 0.5833 1 1>
        <rgba 0.75 0.75 0.75 1>
       ]}}
     Random Seed=?
    )
   ) {isSetParameters=true}]
  emitterUpdateScript [
   &EmitterState(Life Cycle Mode=::NewEnumerator1, Loop Behavior=::NewEnumerator1, Loop Duration Mode=::NewEnumerator1, UseLoopDelay=false)
   &VC_FilteredSocketCount()
   &SpawnRate(SpawnRate=4.75, Use Spawn Probability=false)
   &SpawnRate(SpawnGroup=1, SpawnRate=4.75, Use Spawn Probability=false) {module=SpawnRate001}
   &SpawnRate(SpawnGroup=2, SpawnRate=4.75, Use Spawn Probability=false) {module=SpawnRate002}
  ]
  particleSpawnScript [
   &VC_PaintedVariantsOverride(
    DefaultColor=<rgba 0.125 1 0.5 1>
    ForestGreen=<rgba 0.3304 1.5 0 1>
    Lime=<rgba 0.75 1 0 1>
    Purple=<rgba 0.5 0 1 1>
    Saffron=<rgba 0.75 0.6875 0 2>
    Use R1 Painted Overrides=true
   )
   &InitializeParticle(
    Color=$Particles.Color
    Lifetime Max=&SelectFloatFromArray(
     Direct Array Index=&SpawnGroup()
     Float Selection Array=DataInterface() {propertyValues {FloatData [0.5,0.55,0.65]}}
    )
    Lifetime Min:NiagaraFloat=&SelectFloatFromArray(
     Direct Array Index=&SpawnGroup()
     Float Selection Array=DataInterface() {propertyValues {FloatData [0.35,0.4,0.45]}}
    )
    Lifetime Mode=::NewEnumerator1
    Mesh Scale=&SelectVectorFromArray(
     Direct Array Index=&SpawnGroup()
     Vector Selection Array=DataInterface() {propertyValues {FloatData [<19 19 24>, <15 15 20>]}}
    )
    Mesh Scale Mode=::NewEnumerator4
    Sprite Rotation Mode=::NewEnumerator2
    Sprite Size Mode=::NewEnumerator1
    Uniform Sprite Size Max=200
    Uniform Sprite Size Min=150
    UsePositionOffset=false
    Write Material Random=true
   )
   &PartitionParticles(Number Of Partitions=$Emitter.PortCount, Particles Per Partition=3, Partition Mode:ENiagara_PartitionMode=::NewEnumerator1)
   &SkeletalMeshLocation(
    Bone / Socket Index=$Particles.PartitionParticles.Partition
    Bone / Socket Sampling Mode=::NewEnumerator5
    MaskSpawnGroup=false
    Mesh Sampling Type=::NewEnumerator6
    Sampled Position Offset=&AddVectorToPosition(
     Position=&MakeCustomVectorFromBool(
      Bool=$User.bIsFrontendPreview
      True Vector=&SelectVectorFromArray(
       Direct Array Index=&SpawnGroup()
       Vector Selection Array=DataInterface() {propertyValues {FloatData [<-65 0 0>, <-90 0 0>, <-120 0 0>]}}
      )
      False Vector=&SelectVectorFromArray(
       Direct Array Index=&SpawnGroup()
       Vector Selection Array=DataInterface() {propertyValues {FloatData [<-35 0 0>, <-60 0 0>, <-90 0 0>]}}
      )
     )
     Vector=$User.DTCarOffset
    )
    Sampled Position Offset Coordinate Space:ENiagaraSkelMeshTransforms=::NewEnumerator2
    Skeletal Mesh=$System.SkeletalMesh
   )
   &InitialMeshOrientation(Mesh Orientation Mode=::NewEnumerator7, Rotation=<0 0.75 0>, Use Rotation Vector=true)
   &AddVelocity(Velocity=&RandomRangeVector(Minimum=<-2000 100 -50>, Maximum=<-3000 -100 50>)) {enabled=false}
   /SetVariables(Particles.RenderVisTag:NiagaraInt32=$Emitter.SpawnGroup) {isSetParameters=true}
  ]
  particleUpdateScript [
   &ParticleState()
   &InheritSourceMovement(
    Applied Position Scale=&Multiply_VectorByFloat(
     Vector=<1 1 1>
     Float=&FloatFromCurve(FloatCurve=DataInterface() {propertyValues {Curve {keys [{interpMode=RCIM_Cubic, tangentMode=RCTM_User, time=0.3, value=1}, {interpMode=RCIM_Cubic, tangentMode=RCTM_User, time=1, value=0}]}}})
    )
    ApplyForce=false
    ApplyPosition=true
    ApplyPositionFalloff=false
   ) {enabled=false}
   &SolveForcesAndVelocity(Clamp Velocity=false, Limit Acceleration=false) {enabled=false}
   &DynamicMaterialParameters(
    Index 0 Param 2=1
    Index 0 Param 4=&FloatFromCurve(FloatCurve=DataInterface() {propertyValues {Curve {
        defaultValue=1
        keys [
         {arriveTangent=-0.0626, interpMode=RCIM_Cubic, leaveTangent=-0.0626, tangentMode=RCTM_User, time=0, value=0.95}
         {arriveTangent=-6.0587, interpMode=RCIM_Cubic, leaveTangent=4.8669, tangentMode=RCTM_Break, time=0.5, value=-0.75}
         <at 1 0.95 RCIM_Cubic>
        ]
       }}})
    Index 1 Param 2=1
    Module.Index0Float0Write=true
    Module.Index0Float1Write=true
    Module.Index0Float2Write=true
    Module.Index0Float3Write=true
    Module.Index1Float0Write=true
    Module.Index1Float1Write=true
    Module.Index1Float2Write=true
    Module.Index1Float3Write=true
    Write Parameter Index 1=::NewEnumerator1
   )
   /SetVariables(
    Particles.SoundSprite_DMP_00:Vector4f=&MakeVector4(
     X=&FloatFromCurve(
      FloatCurve=DataInterface() {propertyValues {Curve {defaultValue=0, keys [{tangentMode=RCTM_User, time=0, value=1}, {tangentMode=RCTM_User, time=1, value=0}]}}}
      Scale Curve=0.25
     )
     Y=&FloatFromCurve(
      FloatCurve=DataInterface() {propertyValues {Curve {defaultValue=0, keys [{tangentMode=RCTM_User, time=0, value=1}, {tangentMode=RCTM_User, time=1, value=0}]}}}
      Scale Curve=0.5
     )
     Z=1
     W=2
    )
    Particles.SoundSprite_DMP_01:Vector4f=&MakeVector4(X=0, Y=0, Z=0, W=0)
    Particles.SoundSprite_DMP_02:Vector4f=&MakeVector4(X=2, Y=0, Z=1, W=1)
   ) {isSetParameters=true}
   &ScaleSpriteSize()
   &ScaleColor(
    Scale Alpha=&FloatFromCurve(FloatCurve=DataInterface() {propertyValues {Curve {
        defaultValue=1
        keys [
         {tangentMode=RCTM_User, time=0, value=1}
         {arriveTangent=-0.0405, interpMode=RCIM_Cubic, leaveTangent=-0.0405, tangentMode=RCTM_User, time=0.7009, value=1}
         {arriveTangent=-8.0089, interpMode=RCIM_Cubic, leaveTangent=-8.0089, tangentMode=RCTM_User, time=1, value=0}
        ]
       }}})
    ScaleA=true
    ScaleRGB=true
   )
  ]
  rendererClasses [&NiagaraMeshRendererProperties, &NiagaraSpriteRendererProperties]
  renderers [
   {rendererClass=&NiagaraMeshRendererProperties, rendererIndex=0}
   {rendererClass=&NiagaraMeshRendererProperties, rendererIndex=1}
   {rendererClass=&NiagaraSpriteRendererProperties, rendererIndex=2}
  ]
 }
 NE_Ribbons001 {
  emitterUpdateScript [
   &EmitterState(Life Cycle Mode=::NewEnumerator1, Loop Duration=1, UseLoopDelay=false)
   &VC_FilteredSocketCount()
   &VC_RateQualityHandler()
   &VC_SpawnRibbon_v1(Number of Ribbons=$Emitter.PortCount, SpawnRate=$Emitter.Spawn Rate)
   /SetVariables(
    Emitter.Tex1_Channel:Vector4f=&Vector4FromCurve(
     CurveIndex=$Emitter.NormalizedLoopAge
     Vector4Curve=DataInterface() {propertyValues {
       AlphaCurve {keys [<at 0 0>,<at 1 0>]}
       BlueCurve {keys [<at 0 0>,<at 0.25 1>,<at 0.5 0>,<at 0.75 1>,<at 1 0>]}
       GreenCurve {keys [<at 0 0>,<at 1 0>]}
       RedCurve {defaultValue=1, keys [<at 0 1>,<at 0.25 0>,<at 0.5 1>,<at 0.75 0>,<at 1 1>]}
      }}
    )
    Emitter.Tex2_Channel:Vector4f=&Vector4FromCurve(
     CurveIndex=$Emitter.NormalizedLoopAge
     Vector4Curve=DataInterface() {propertyValues {
       AlphaCurve {keys [<at 0 0>,<at 1 0>]}
       BlueCurve {defaultValue=0, keys []}
       GreenCurve {defaultValue=0, keys [<at 0 1>,<at 0.25 0>,<at 0.5 1>,<at 0.75 0>,<at 1 1>]}
       RedCurve {defaultValue=1, keys [<at 0 0>,<at 0.25 1>,<at 0.5 0>,<at 0.75 1>,<at 1 0>]}
      }}
    )
   ) {isSetParameters=true}
  ]
  particleSpawnScript [
   &VC_PaintedVariantsOverride() {enabled=false}
   &VC_InitializeRibbon_v1(
    Color=$Particles.Color
    Custom Offset=true
    Enable Skeletal Position=true
    Lifetime=&MakeCustomFloatFromBool(Bool=$User.bIsFrontendPreview, True Float=0.65, False Float=0.5)
    Local Vector Offset=&AddVectorToPosition(
     Position=/Niagara/DynamicInputs/TypeConversions/MakeCustomPositionFromBool(Bool:NiagaraBool=$User.bIsFrontendPreview, True Position:NiagaraPosition=<-60 0 0>, False Position:NiagaraPosition=<-60 0 0>)
     Vector=$User.DTCarOffset
    )
    Offset Amount=1
    Ribbon Width=11
   )
   &AddVelocity(Velocity=&MakeCustomVectorFromBool(Bool=$User.bIsFrontendPreview, True Vector=<-1500 0 1>, False Vector=<-1500 0 0>)) {module=AddVelocity001}
  ]
  particleUpdateScript [
   &ParticleState()
   &InheritSourceMovement(
    Applied Position Scale=&Multiply_VectorByFloat(
     Vector=<1 1 1>
     Float=&FloatFromCurve(FloatCurve=DataInterface() {propertyValues {Curve {keys [
          {
           arriveTangentWeight=0.4252
           interpMode=RCIM_Cubic
           leaveTangentWeight=0.4252
           tangentMode=RCTM_User
           tangentWeightMode=RCTWM_WeightedBoth
           time=0.1
           value=1
          }
          {interpMode=RCIM_Cubic, tangentMode=RCTM_User, time=1, value=0}
         ]}}})
    )
    ApplyForce=false
    ApplyPosition=true
    ApplyPositionFalloff=false
   )
   &VC_PaintedVariantsOverride(
    Black=&ColorFromCurve(ColorCurve=DataInterface() {propertyValues {
       AlphaCurve {keys [<at 0 1 RCIM_Cubic>,<at 1 2 RCIM_Cubic>]}
       BlueCurve {keys [<at 0 0.401 RCIM_Cubic>,<at 0.3488 0>,<at 1 0 RCIM_Cubic>]}
       GreenCurve {keys [<at 0 0.401 RCIM_Cubic>,<at 0.3488 0>,<at 1 0 RCIM_Cubic>]}
       RedCurve {keys [<at 0 0.401 RCIM_Cubic>,<at 0.3488 0>,<at 1 0 RCIM_Cubic>]}
      }})
    BurntSienna=&ColorFromCurve(ColorCurve=DataInterface() {propertyValues {
       AlphaCurve {keys [<at 0 1 RCIM_Cubic>,<at 0.4601 1>,<at 1 1.5 RCIM_Cubic>]}
       BlueCurve {keys [<at 0 1.487 RCIM_Cubic>,<at 0.1938 0.2>,<at 0.3965 0.1 RCIM_Cubic>]}
       GreenCurve {keys [<at 0 2.4957 RCIM_Cubic>,<at 0.1938 0.3568>,<at 0.3965 0.2 RCIM_Cubic>]}
       RedCurve {keys [<at 0 3 RCIM_Cubic>,<at 0.1938 1>,<at 0.3965 0.4 RCIM_Cubic>]}
      }})
    Cobalt=&ColorFromCurve(ColorCurve=DataInterface() {propertyValues {
       AlphaCurve {keys [<at 0 1 RCIM_Cubic>,<at 0.5536 1>,<at 1 2 RCIM_Cubic>]}
       BlueCurve {keys [<at 0 3 RCIM_Cubic>, {arriveTangent=-3.876, interpMode=RCIM_Cubic, leaveTangent=-3.876, time=0.2912, value=0.7}, <at 0.645 0.5>]}
       GreenCurve {keys [<at 0 1.3284 RCIM_Cubic>, {arriveTangent=-1.9319, interpMode=RCIM_Cubic, leaveTangent=-1.9319, time=0.2912, value=0.1154}, <at 0.645 0.0824>]}
       RedCurve {keys [<at 0 0.0395 RCIM_Cubic>, {arriveTangent=-0.0612, interpMode=RCIM_Cubic, leaveTangent=-0.0612, time=0.2912, value=0}, <at 0.645 0>]}
      }})
    Crimson=&ColorFromCurve(ColorCurve=DataInterface() {propertyValues {
       AlphaCurve {keys [<at 0.6483 1 RCIM_Cubic>,<at 1 2 RCIM_Cubic>]}
       BlueCurve {keys [<at 0.0879 0.75 RCIM_Cubic>,<at 0.3794 0.0167>,<at 0.7516 0>]}
       GreenCurve {keys [<at 0.0879 0.75 RCIM_Cubic>,<at 0.3794 0>,<at 0.7516 0.01>]}
       RedCurve {keys [<at 0.0879 3 RCIM_Cubic>,<at 0.3794 1>,<at 0.7516 0.25>]}
      }})
    DefaultColor=&ColorFromCurve(ColorCurve=DataInterface() {propertyValues {
       AlphaCurve {keys [<at 0 0 RCIM_Cubic>,<at 0.0894 1>,<at 0.8757 1>,<at 1 0 RCIM_Cubic>]}
       BlueCurve {keys [<at 0 1.5 RCIM_Cubic>,<at 0.1276 1>,<at 0.3914 0.2>]}
       GreenCurve {keys [<at 0 1.2506 RCIM_Cubic>,<at 0.1276 0.6667>,<at 0.3914 1>]}
       RedCurve {keys [<at 0 0.15 RCIM_Cubic>,<at 0.1276 0>,<at 0.3914 0>]}
      }})
    ForestGreen=&ColorFromCurve(ColorCurve=DataInterface() {propertyValues {
       AlphaCurve {keys [<at 0.6642 1 RCIM_Cubic>,<at 1 6.5 RCIM_Cubic>]}
       BlueCurve {keys [<at 0.1933 0 RCIM_Cubic>,<at 0.4033 0>,<at 0.6014 0 RCIM_Cubic>]}
       GreenCurve {keys [<at 0.1933 1.25 RCIM_Cubic>,<at 0.4033 0.75>,<at 0.6014 0.32 RCIM_Cubic>]}
       RedCurve {keys [<at 0.1933 0.5 RCIM_Cubic>,<at 0.4033 0.0048>,<at 0.6014 0 RCIM_Cubic>]}
      }})
    Grey=&ColorFromCurve(ColorCurve=DataInterface() {propertyValues {
       AlphaCurve {keys [<at 0.6642 1 RCIM_Cubic>,<at 1 5 RCIM_Cubic>]}
       BlueCurve {keys [<at 0.0716 1 RCIM_Cubic>,<at 0.4425 0.175 RCIM_Cubic>]}
       GreenCurve {keys [<at 0.0716 1 RCIM_Cubic>,<at 0.4425 0.175 RCIM_Cubic>]}
       RedCurve {keys [<at 0.0716 1 RCIM_Cubic>,<at 0.4425 0.175 RCIM_Cubic>]}
      }})
    Lime=&ColorFromCurve(ColorCurve=DataInterface() {propertyValues {
       AlphaCurve {keys [<at 0.678 1 RCIM_Cubic>,<at 1 8 RCIM_Cubic>]}
       BlueCurve {keys [<at 0 0 RCIM_Cubic>,<at 0.5092 0 RCIM_Cubic>]}
       GreenCurve {keys [<at 0 2 RCIM_Cubic>,<at 0.5092 0.45 RCIM_Cubic>]}
       RedCurve {keys [<at 0 1.3333 RCIM_Cubic>,<at 0.5092 0.3375 RCIM_Cubic>]}
      }})
    Orange=&ColorFromCurve(ColorCurve=DataInterface() {propertyValues {
       AlphaCurve {keys [<at 0.7133 1 RCIM_Cubic>,<at 1 6 RCIM_Cubic>]}
       BlueCurve {keys [<at 0.0422 0 RCIM_Cubic>,<at 0.5171 0 RCIM_Cubic>]}
       GreenCurve {keys [<at 0.0422 1.4057 RCIM_Cubic>,<at 0.5171 0.3058 RCIM_Cubic>]}
       RedCurve {keys [<at 0.0422 2 RCIM_Cubic>,<at 0.5171 1 RCIM_Cubic>]}
      }})
    Pink=&ColorFromCurve(ColorCurve=DataInterface() {propertyValues {
       AlphaCurve {keys [<at 0.7035 1 RCIM_Cubic>,<at 1 4 RCIM_Cubic>]}
       BlueCurve {keys [<at 0 1 RCIM_Cubic>,<at 0.3679 0.45 RCIM_Cubic>]}
       GreenCurve {keys [<at 0 1 RCIM_Cubic>,<at 0.3679 0 RCIM_Cubic>]}
       RedCurve {keys [<at 0 1 RCIM_Cubic>,<at 0.3679 1 RCIM_Cubic>]}
      }})
    Purple=&ColorFromCurve(ColorCurve=DataInterface() {propertyValues {
       AlphaCurve {keys [<at 0.7388 1 RCIM_Cubic>,<at 1 6 RCIM_Cubic>]}
       BlueCurve {keys [<at 0 2.9138 RCIM_Cubic>,<at 0.3758 0.65 RCIM_Cubic>]}
       GreenCurve {keys [<at 0 0.6945 RCIM_Cubic>,<at 0.3758 0 RCIM_Cubic>]}
       RedCurve {keys [<at 0 3 RCIM_Cubic>,<at 0.3758 0.4835 RCIM_Cubic>]}
      }})
    Saffron=&ColorFromCurve(ColorCurve=DataInterface() {propertyValues {
       AlphaCurve {keys [<at 0.7074 1 RCIM_Cubic>,<at 1 5 RCIM_Cubic>]}
       BlueCurve {keys [<at 0 0 RCIM_Cubic>,<at 0.3993 0.05 RCIM_Cubic>]}
       GreenCurve {keys [<at 0 4.5833 RCIM_Cubic>,<at 0.3993 0.8417 RCIM_Cubic>]}
       RedCurve {keys [<at 0 5 RCIM_Cubic>,<at 0.3993 1 RCIM_Cubic>]}
      }})
    SkyBlue=&ColorFromCurve(ColorCurve=DataInterface() {propertyValues {
       AlphaCurve {keys [<at 0.7231 1 RCIM_Cubic>,<at 1 5 RCIM_Cubic>]}
       BlueCurve {keys [<at 0.1403 4 RCIM_Cubic>,<at 0.4346 0.5 RCIM_Cubic>]}
       GreenCurve {keys [<at 0.1403 2.3333 RCIM_Cubic>,<at 0.4346 0.2917 RCIM_Cubic>]}
       RedCurve {keys [<at 0.1403 0 RCIM_Cubic>,<at 0.4346 0 RCIM_Cubic>]}
      }})
    TitaniumWhite=&ColorFromCurve(ColorCurve=DataInterface() {propertyValues {
       AlphaCurve {keys [<at 0.7251 1 RCIM_Cubic>,<at 1 5 RCIM_Cubic>]}
       BlueCurve {keys [<at 0 4 RCIM_Cubic>,<at 0.4189 1 RCIM_Cubic>]}
       GreenCurve {keys [<at 0 4 RCIM_Cubic>,<at 0.4189 1 RCIM_Cubic>]}
       RedCurve {keys [<at 0 4 RCIM_Cubic>,<at 0.4189 1 RCIM_Cubic>]}
      }})
   ) {module=VC_PaintedVariantsOverride001}
   &SolveForcesAndVelocity(Clamp Velocity=false, Limit Acceleration=false)
   &VC_UpdateRibbon_v1(Enable Skeletal Position=true)
   /SetVariables(Particles.Position=&AddVectorToPosition(
     Position=$Particles.Position
     Vector=&TransformVector(
      Vector=&MakeVector(
       X=&Cosine(
        Normalized Angle=&Add_Float(
         A=&Multiply_Float(A=$Particles.NormalizedAge, B=-1.15)
         B=&Multiply_Float(A=$Emitter.LoopedAge, B=1.5)
        )
        Period=1
        Scale=&FloatFromCurve(
         FloatCurve=DataInterface() {propertyValues {Curve {keys [
             {arriveTangent=12.1602, interpMode=RCIM_Cubic, leaveTangent=12.1602, tangentMode=RCTM_User, time=0, value=0.5}
             {arriveTangent=0.004, interpMode=RCIM_Cubic, leaveTangent=0.004, tangentMode=RCTM_User, time=0.065, value=5}
             {arriveTangent=-2.5532, interpMode=RCIM_Cubic, leaveTangent=-2.5532, tangentMode=RCTM_User, time=0.2, value=1}
             <at 0.65 0 RCIM_Cubic>
            ]}}}
         Scale Curve=&SelectFloatFromArray(Direct Array Index=&SpawnGroup(), Float Selection Array=DataInterface() {propertyValues {FloatData [1,-1]}})
        )
        Bias=0
       )
       Y=&Sine(
        Normalized Angle=&Add_Float(
         A=&Multiply_Float(A=$Particles.NormalizedAge, B=-1.15)
         B=&Multiply_Float(A=$Emitter.LoopedAge, B=1.5)
        )
        Period=1
        Scale=&FloatFromCurve(
         FloatCurve=DataInterface() {propertyValues {Curve {keys [
             {arriveTangent=12.1602, interpMode=RCIM_Cubic, leaveTangent=12.1602, tangentMode=RCTM_User, time=0, value=0.5}
             {arriveTangent=0.004, interpMode=RCIM_Cubic, leaveTangent=0.004, tangentMode=RCTM_User, time=0.065, value=5}
             {arriveTangent=-2.5532, interpMode=RCIM_Cubic, leaveTangent=-2.5532, tangentMode=RCTM_User, time=0.2, value=1}
             <at 0.65 0 RCIM_Cubic>
            ]}}}
         Scale Curve=&SelectFloatFromArray(Direct Array Index=&SpawnGroup(), Float Selection Array=DataInterface() {propertyValues {FloatData [1,-1]}})
        )
        Bias=0
       )
       Z=&Sine(
        Normalized Angle=&Add_Float(
         A=&Multiply_Float(A=$Particles.NormalizedAge, B=-1.15)
         B=&Multiply_Float(A=$Emitter.LoopedAge, B=2)
        )
        Period=1
        Scale=&FloatFromCurve(
         FloatCurve=DataInterface() {propertyValues {Curve {keys [
             {arriveTangent=12.1602, interpMode=RCIM_Cubic, leaveTangent=12.1602, tangentMode=RCTM_User, time=0, value=0.5}
             {arriveTangent=0.004, interpMode=RCIM_Cubic, leaveTangent=0.004, tangentMode=RCTM_User, time=0.065, value=5}
             {arriveTangent=-2.5532, interpMode=RCIM_Cubic, leaveTangent=-2.5532, tangentMode=RCTM_User, time=0.2, value=1}
             <at 0.65 0 RCIM_Cubic>
            ]}}}
         Scale Curve=&SelectFloatFromArray(Direct Array Index=&SpawnGroup(), Float Selection Array=DataInterface() {propertyValues {FloatData [2,-2]}})
        )
        Bias=0
       )
      )
      Destination Space=::Simulation
      Apply Scale=true
     )
    )) {isSetParameters=true}
   &DynamicMaterialParameters(Write Parameter Index 0=false) {enabled=false}
   &ScaleRibbonWidth(Ribbon Width Scale=&FloatFromCurve(
     FloatCurve=DataInterface() {propertyValues {Curve {keys [
         {arriveTangent=11.3893, interpMode=RCIM_Cubic, leaveTangent=11.3893, tangentMode=RCTM_User, time=0, value=0.25}
         {arriveTangent=-0.0905, interpMode=RCIM_Cubic, leaveTangent=-0.0905, tangentMode=RCTM_User, time=0.1, value=1}
         <at 1 0.5 RCIM_Cubic>
        ]}}}
     Scale Curve=5
    ))
  ]
  rendererClasses [&NiagaraRibbonRendererProperties]
  renderers [{rendererClass=&NiagaraRibbonRendererProperties}]
 }
 NE_TrailMeshes {
  emitterSpawnScript [/SetVariables(
    Emitter.R0_BaseColor:LinearColor=&SelectLinearColorFromArray(
     Array Sampling Mode=::NewEnumerator1
     Direct Array Index=$System.ColorVariantID
     Color Selection Array=DataInterface() {propertyValues {ColorData [
        <rgba 0.4364 0.8893 0.9 1>
        <rgba 1 1 1 1>
        <rgba 0.01 0.01 0.01 1>
        <rgba 1 0.45 0.2 1>
        <rgba 0.067 0.2917 1 1>
        <rgba 1 0 0.0167 1>
        <rgba 0.2 1.5 0.1 1>
        <rgba 1 1 1 1>
        <rgba 0.4653 1 0 1>
        <rgba 0.8 0.2933 0.04 1>
        <rgba 1 0.1 0.7 1>
        <rgba 0.275 0.05 0.5 1>
        <rgba 1 0.9167 0 1>
        <rgba 0 0.1913 0.25 1>
        <rgba 0.25 0.25 0.25 1>
       ]}}
     Random Seed=?
    )
    Emitter.R0_DissolveHighlightColor:LinearColor=&SelectLinearColorFromArray(
     Array Sampling Mode=::NewEnumerator1
     Direct Array Index=$System.ColorVariantID
     Color Selection Array=DataInterface() {propertyValues {ColorData [
        <rgba 0.6 1 1 -0.5>
        <rgba 1 1 1 1>
        <rgba 0.01 0.01 0.01 -0.5>
        <rgba 1 0.45 0.2 -0.5>
        <rgba 0.067 0.2917 1 -0.5>
        <rgba 1 0 0.0167 0>
        <rgba 0.12 2 0.1 -0.5>
        <rgba 0.25 0.25 0.25 -0.5>
        <rgba 0.4653 1 0 -0.5>
        <rgba 1 0.4167 0 0>
        <rgba 1 0.1 0.7 0>
        <rgba 0.275 0.05 0.5 0>
        <rgba 0.97 0.8892 0 0>
        <rgba 0 0.5 0.5 0>
        <rgba 0.75 0.75 0.75 0>
       ]}}
     Random Seed=?
    )
    Emitter.R0_FresnelColor:LinearColor=&SelectLinearColorFromArray(
     Array Sampling Mode=::NewEnumerator1
     Direct Array Index=$System.ColorVariantID
     Color Selection Array=DataInterface() {propertyValues {ColorData [
        <rgba 0.0294 0.25 0.25 0>
        <rgba 1 1 1 1>
        <rgba 0.005 0.005 0.005 0>
        <rgba 0.2 0.1 0.05 0>
        <rgba 0.0302 0.1001 0.45 0>
        <rgba 0.55 0.005 0.005 0>
        <rgba 0.01 0.5 0.005 0>
        <rgba 0.01 0.01 0.01 0>
        <rgba 0.3386 0.5 0.05 0>
        <rgba 0.8 0.3473 0.024 0>
        <rgba 0.85 0.0425 0.5808 0>
        <rgba 0.3333 0 1 0>
        <rgba 0.5 0.4596 0.015 0>
        <rgba 0 0.5 0.5 0>
        <rgba 0.75 0.75 0.75 0>
       ]}}
     Random Seed=?
    )
   ) {isSetParameters=true}]
  emitterUpdateScript [
   &EmitterState(
    Life Cycle Mode=::NewEnumerator1
    Loop Delay=&RandomRangeFloat(Minimum=0.2, Maximum=0.3)
    Loop Duration=1
    UseLoopDelay=true
   )
   &VC_FilteredSocketCount()
   &SpawnRate(
    SpawnRate=&MakeCustomFloatFromBool(Bool=$User.bIsFrontendPreview, True Float=&RandomRangeFloat(Minimum=20, Maximum=30), False Float=40)
    Use Spawn Probability=false
   )
  ]
  particleSpawnScript [
   &VC_PaintedVariantsOverride(DefaultColor=<rgba 2 2 2 1>)
   &InitializeParticle(
    Color=$Particles.Color
    Lifetime Max=0.45
    Lifetime Min=0.3
    Lifetime Mode=::NewEnumerator1
    Mesh Scale Mode=::NewEnumerator3
    Mesh Uniform Scale=&MakeCustomFloatFromBool(Bool=$User.bIsFrontendPreview, True Float=8, False Float=8)
    Position Offset=<5 0 0>
    Sprite Rotation Mode=::NewEnumerator2
    UsePositionOffset=true
    Write Material Random=true
   )
   &PartitionParticles(Number Of Partitions=$Emitter.PortCount)
   &SkeletalMeshLocation(
    Bone / Socket Index=$Particles.PartitionParticles.Partition
    Bone / Socket Sampling Mode=::NewEnumerator5
    MaskSpawnGroup=false
    Mesh Sampling Type=::NewEnumerator6
    Sampled Position Offset=&AddVectorToPosition(Position=<-55 0 0>, Vector=$User.DTCarOffset)
    Sampled Position Offset Coordinate Space:ENiagaraSkelMeshTransforms=::NewEnumerator2
    Skeletal Mesh=$System.SkeletalMesh
   )
   &InitialMeshOrientation(Mesh Orientation Mode=::NewEnumerator7, Rotation Coordinate Space=::NewEnumerator0, Use Rotation Vector=true)
   &AddVelocity(Velocity=<-1200 0 0>)
  ]
  particleUpdateScript [
   &ParticleState()
   &ScaleMeshSize(Scale Factor=&Multiply_VectorByFloat(
     Vector=<1 1 1>
     Float=&FloatFromCurve(
      FloatCurve=DataInterface() {propertyValues {Curve {keys [
          {arriveTangent=9.0725, interpMode=RCIM_Cubic, leaveTangent=9.0725, tangentMode=RCTM_User, time=0, value=0}
          {arriveTangent=-1.3753, interpMode=RCIM_Cubic, leaveTangent=-1.3753, tangentMode=RCTM_User, time=0.1408, value=0.85}
          {arriveTangent=-0.0405, interpMode=RCIM_Cubic, leaveTangent=-0.0405, tangentMode=RCTM_User, time=1, value=0.35}
         ]}}}
      Scale Curve=2
     )
    ))
   &InheritSourceMovement(
    Applied Position Scale=&Multiply_VectorByFloat(
     Vector=<1 1 1>
     Float=&FloatFromCurve(FloatCurve=DataInterface() {propertyValues {Curve {keys [
          {
           arriveTangentWeight=0.4252
           interpMode=RCIM_Cubic
           leaveTangentWeight=0.4252
           tangentMode=RCTM_User
           tangentWeightMode=RCTWM_WeightedBoth
           time=0.5
           value=1
          }
          {interpMode=RCIM_Cubic, tangentMode=RCTM_User, time=1, value=0}
         ]}}})
    )
    ApplyForce=false
    ApplyPosition=true
    ApplyPositionFalloff=false
   )
   &InheritVelocity(
    Inherited Velocity Amount Scale=&MakeCustomVectorFromBool(Bool=$User.bIsFrontendPreview, True Vector=<0 0 0>, False Vector=<1 1 1>)
    Inherited Velocity Speed Limit=-100
    Speed Limit=true
    UseSpeedThreshold=false
    Velocity Source Space:ENiagaraCoordinateSpace=::Simulation
   ) {enabled=false}
   &SolveForcesAndVelocity(Clamp Velocity=false, Limit Acceleration=false)
   &DynamicMaterialParameters(
    Index 0 Param 2=2
    Index 0 Param 4=&FloatFromCurve(FloatCurve=DataInterface() {propertyValues {Curve {
        defaultValue=1
        keys [
         {arriveTangent=0.0465, interpMode=RCIM_Cubic, leaveTangent=0.0465, tangentMode=RCTM_User, time=0, value=0.95}
         {arriveTangent=-2.6624, interpMode=RCIM_Cubic, leaveTangent=4.8669, tangentMode=RCTM_Break, time=0.5, value=-0.2}
         <at 1 0.95 RCIM_Cubic>
        ]
       }}})
    Index 1 Param 2=1
    Module.Index0Float0Write=true
    Module.Index0Float1Write=true
    Module.Index0Float2Write=true
    Module.Index0Float3Write=true
    Module.Index1Float0Write=true
    Module.Index1Float1Write=true
    Module.Index1Float2Write=true
    Module.Index1Float3Write=true
    Write Parameter Index 1=::NewEnumerator1
   )
   &ScaleColor(
    Scale Alpha=&FloatFromCurve(FloatCurve=DataInterface() {propertyValues {Curve {keys [<at 0.5 1>,<at 1 0>]}}}, Scale Curve=0.65)
    ScaleA=true
    ScaleRGB=true
   ) {enabled=false}
  ]
  rendererClasses [&NiagaraMeshRendererProperties, &NiagaraSpriteRendererProperties]
  renderers [
   {rendererClass=&NiagaraMeshRendererProperties, rendererIndex=0}
   {rendererClass=&NiagaraSpriteRendererProperties, rendererIndex=1}
  ]
 }
}
summary {
 systemName=NS_Boost_BassBoost_OP
 emitters [
  {emitterName=NE_TrailMeshes, rendererClasses [&NiagaraMeshRendererProperties, &NiagaraSpriteRendererProperties]}
  {
   emitterName=NE_BoostCone_Meshes
   rendererClasses [&NiagaraMeshRendererProperties, &NiagaraSpriteRendererProperties]
  }
  {emitterName=NE_Ribbons001, rendererClasses [&NiagaraRibbonRendererProperties]}
 ]
 userVariables [
  User.bIsFrontendPreview {type=NiagaraBool}
  User.bIsPainted {type=NiagaraBool}
  User.ColorVariantID {type=NiagaraInt32}
  User.Desaturation {type=NiagaraFloat, defaultValue=0.15, description=""}
  User.DTCarOffset {type=Vector3f, description="", defaultValue {struct=Vector3f, value=<0 0 0>}}
  User.GlowOverride {type=NiagaraFloat, defaultValue=4, description=""}
  User.PaintedColor {type=LinearColor, description="", defaultValue {struct=LinearColor, value=<rgba 1 1 1 1>}}
  User.PaintedColor_2 {type=LinearColor, description="", defaultValue {struct=LinearColor, value=<rgba 1 1 1 1>}}
  User.PaintedColor_3 {type=LinearColor, description="", defaultValue {struct=LinearColor, value=<rgba 1 1 1 1>}}
  User.PaintedColor_4 {type=LinearColor, description="", defaultValue {struct=LinearColor, value=<rgba 1 1 1 1>}}
  User.PaintedColor_5 {type=LinearColor, description="", defaultValue {struct=LinearColor, value=<rgba 1 1 1 1>}}
  User.PaintedScalar {type=NiagaraFloat, defaultValue=1, description=""}
  User.PortCount {type=NiagaraInt32, defaultValue=1}
  User.RingsEmissive {type=NiagaraFloat, defaultValue=6, description=""}
 ]
}
